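# Niagara cosmetic effect for the Banana Philosopher player body that emits occasional short-lived golden sparkles around the character.
# Sparkles spawn by random bursts and probabilistic rate, then follow sampled skeletal mesh positions within a small rotation-aligned sphere.
# Sprites pop and shrink, with three renderers and derived glow, outline and front parameters for a layered look using time-of-day colors.
# A disabled Sparkles001 variant uses fixed colors and one renderer.
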NiagaraSystem /BRCosmetics/Characters/Player/Male/Medium/Bodies/M_MED_BananaPhilosopher/FX/NS_BananaPhilosopher
dependencies {
 usedDataInterfaces [
  &NiagaraDataInterfaceArrayFloat
  &NiagaraDataInterfaceColorCurve
  &NiagaraDataInterfaceCurve
  &NiagaraDataInterfaceSkeletalMesh
 ]
 usedDynamicInputs [&SelectFloatFromArray, &FloatFromCurve]
 usedModules [
  &self:SystemSpawnScript.SystemScriptSource.SystemScriptGraph.NiagaraNodeAssignment_0.SetVariables_#1
  &SystemState
 ]
}
emitters {
 Sparkles {
  emitterUpdateScript [
   &EmitterState()
   &SpawnRate(SpawnRate=10, Use Spawn Probability=false) {enabled=false, module=SpawnRate001}
   &SpawnBurst_Instantaneous(
    Spawn Count=&RandomRangeInt(Minimum=0, Maximum=1)
    Spawn Time=&RandomRangeFloat(Minimum=0.5, Maximum=2)
    Use Spawn Probability=false
    UseLoopCountLimit=false
   )
   &SpawnRate(
    Spawn Probability=0.75
    SpawnRate=&RandomRangeFloat(Minimum=0.65, Maximum=1.35)
    Use Spawn Probability=true
   )
  ]
  particleSpawnScript [
   &InitializeParticle(
    Color=&FNTimeOfDayColor(
     Night Color=<rgba 1.25 0.9347 0 1>
     Morning Color=<rgba 2.5 1.8693 0 1>
     Day Color=<rgba 2.5 1.8693 0 1>
     Evening Color=<rgba 2.5 1.8693 0 1>
    )
    Lifetime Max=0.75
    Lifetime Min=0.5
    Lifetime Mode=::NewEnumerator1
    Position Mode=::NewEnumerator0
    Sprite Size Mode=::NewEnumerator1
    Uniform Sprite Size Max=4
    Uniform Sprite Size Min=2.5
    Write Material Random=true
   )
   &ApplyInitialForces()
   &FN_SetSpawnGroup(Number of Spawn Groups=2)
   /SetVariables(Particles.GlowSize=&Vector2DFromFloat(Value=20)) {isSetParameters=true, module=SetVariables_#2}
   &CameraOffset(Camera Offset Amount=15)
  ]
  particleUpdateScript [
   &ParticleState()
   &SkeletalMeshLocation(
    MaskSpawnGroup=false
    Mesh Sampling Type=::NewEnumerator5
    Position Sampling=::NewEnumerator1
    Sampled Position Offset=<-5 -2 0>
    Skeletal Mesh=$System.Skeletal Mesh
    Socket Sampling Mode=::NewEnumerator5
   )
   &ShapeLocation(
    Hemisphere Distribution=<1 0.4>
    MaskSpawnGroup=false
    Override Seed=false
    Rotation Axis=&ReturnMeshOrientationAxis(Mesh Orientation Axis=::XAxis, Mesh Orientation Quat=$Particles.SkeletalMeshLocation.SampledRotation)
    Rotation Mode=::NewEnumerator1
    Shape Origin=$Particles.SkeletalMeshLocation.SampledPosition
    Sphere Radius=8
    UseOwnerScale=true
   )
   /SetVariables(Particles.GlowColor=&ColorFromCurve(ColorCurve=DataInterface() {propertyValues {
       AlphaCurve {keys [<at 0 0 RCIM_Cubic>,<at 0.1158 0.15>,<at 0.5626 0.15>,<at 1 0 RCIM_Cubic>]}
       BlueCurve {keys [<at 0 1 RCIM_Cubic>,<at 1 1 RCIM_Cubic>]}
       GreenCurve {keys [<at 0 1 RCIM_Cubic>,<at 1 1 RCIM_Cubic>]}
       RedCurve {keys [<at 0 1 RCIM_Cubic>,<at 1 1 RCIM_Cubic>]}
      }})) {isSetParameters=true, module=SetVariables_#3}
   &ScaleSpriteSize(
    Uniform Curve Scale=3
    Uniform Curve Sprite Scale=DataInterface() {propertyValues {Curve {keys [
        {arriveTangent=9.0725, interpMode=RCIM_Cubic, leaveTangent=9.0725, tangentMode=RCTM_User, time=0, value=0}
        {arriveTangent=-2.5249, interpMode=RCIM_Cubic, leaveTangent=-2.5249, tangentMode=RCTM_User, time=0.1408, value=0.9847}
        {arriveTangent=-0.0405, interpMode=RCIM_Cubic, leaveTangent=-0.0405, tangentMode=RCTM_User, time=1, value=0}
       ]}}}
   ) {module=ScaleSpriteSize001}
   /SetVariables(
    Particles.Outline Cam Offset:NiagaraFloat=&Add_Float(A=$Particles.CameraOffset, B=-0.1)
    Particles.Outline Size:Vector2f=&Multiply_Vector2D_ByFloat(Float=1.1, Vector2D=$Particles.SpriteSize)
    Particles.Outline color:LinearColor=<rgba 0.5 0.3086 0.1428 1>
   ) {isSetParameters=true, module=SetVariables_#4}
   /SetVariables(
    Particles.FrontCam Offset:NiagaraFloat=&Add_Float(A=0.1, B=$Particles.CameraOffset)
    Particles.FrontSize:Vector2f=&Multiply_Vector2D_ByFloat(Float=0.8, Vector2D=$Particles.SpriteSize)
    Particles.Frontcolor:LinearColor=&FNTimeOfDayColor(Night Color=<rgba 1.75 1.75 0.4375 1>, Morning Color=<rgba 1 1 0.25 1>, Day Color=<rgba 1 1 0.25 1>, Evening Color=<rgba 1 1 0.25 1>)
   ) {isSetParameters=true, module=SetVariables_#5}
  ]
  rendererClasses [&NiagaraSpriteRendererProperties]
  renderers [
   {rendererClass=&NiagaraSpriteRendererProperties}
   {rendererClass=&NiagaraSpriteRendererProperties, rendererIndex=1}
   {rendererClass=&NiagaraSpriteRendererProperties, rendererIndex=2}
  ]
 }
 Sparkles001 {
  bEnabled=false
  emitterUpdateScript [
   &EmitterState()
   &SpawnBurst_Instantaneous(
    Spawn Count=&RandomRangeInt(Minimum=0, Maximum=1)
    Spawn Time=&RandomRangeFloat(Minimum=0.5, Maximum=2)
    Use Spawn Probability=false
    UseLoopCountLimit=false
   )
   &SpawnRate(
    Spawn Probability=0.75
    SpawnRate=&RandomRangeFloat(Minimum=0.65, Maximum=1.35)
    Use Spawn Probability=true
   )
   &SpawnRate(SpawnRate=10, Use Spawn Probability=false) {module=SpawnRate001}
  ]
  particleSpawnScript [
   &InitializeParticle(
    Color=<rgba 10 7.9819 2 1>
    Lifetime Max=0.75
    Lifetime Min=0.25
    Lifetime Mode=::NewEnumerator1
    Position Mode=::NewEnumerator0
    Sprite Size Mode=::NewEnumerator1
    Uniform Sprite Size Max=3.5
    Uniform Sprite Size Min=2.5
    Write Material Random=true
   )
   &ApplyInitialForces()
   &FN_SetSpawnGroup(Number of Spawn Groups=2)
   /SetVariables(Particles.GlowSize=&Vector2DFromFloat(Value=20)) {isSetParameters=true, module=SetVariables_#2}
   &CameraOffset(Camera Offset Amount=15)
  ]
  particleUpdateScript [
   &ParticleState()
   &SkeletalMeshLocation(
    MaskSpawnGroup=false
    Mesh Sampling Type=::NewEnumerator5
    Position Sampling=::NewEnumerator1
    Sampled Position Offset=<-5 -2 0>
    Skeletal Mesh=$System.Skeletal Mesh
    Socket Sampling Mode=::NewEnumerator5
   )
   &ShapeLocation(
    Hemisphere Distribution=<1 0.4>
    MaskSpawnGroup=false
    Override Seed=false
    Rotation Axis=&ReturnMeshOrientationAxis(Mesh Orientation Axis=::XAxis, Mesh Orientation Quat=$Particles.SkeletalMeshLocation.SampledRotation)
    Rotation Mode=::NewEnumerator1
    Shape Origin=$Particles.SkeletalMeshLocation.SampledPosition
    Sphere Radius=8
    UseOwnerScale=true
   )
   /SetVariables(Particles.GlowColor=&ColorFromCurve(ColorCurve=DataInterface() {propertyValues {
       AlphaCurve {keys [<at 0 0 RCIM_Cubic>,<at 0.1158 0.15>,<at 0.5626 0.15>,<at 1 0 RCIM_Cubic>]}
       BlueCurve {keys [<at 0 1 RCIM_Cubic>,<at 1 1 RCIM_Cubic>]}
       GreenCurve {keys [<at 0 1 RCIM_Cubic>,<at 1 1 RCIM_Cubic>]}
       RedCurve {keys [<at 0 1 RCIM_Cubic>,<at 1 1 RCIM_Cubic>]}
      }})) {isSetParameters=true, module=SetVariables_#3}
   &ScaleSpriteSize(
    Uniform Curve Scale=3
    Uniform Curve Sprite Scale=DataInterface() {propertyValues {Curve {keys [
        {arriveTangent=9.0725, interpMode=RCIM_Cubic, leaveTangent=9.0725, tangentMode=RCTM_User, time=0, value=0}
        {arriveTangent=-2.5249, interpMode=RCIM_Cubic, leaveTangent=-2.5249, tangentMode=RCTM_User, time=0.1408, value=0.9847}
        {arriveTangent=-0.0405, interpMode=RCIM_Cubic, leaveTangent=-0.0405, tangentMode=RCTM_User, time=1, value=0}
       ]}}}
   ) {module=ScaleSpriteSize001}
   /SetVariables(
    Particles.Outline Cam Offset:NiagaraFloat=&Add_Float(A=$Particles.CameraOffset, B=-0.1)
    Particles.Outline Size:Vector2f=&Multiply_Vector2D_ByFloat(Float=1.1, Vector2D=$Particles.SpriteSize)
    Particles.Outline color:LinearColor=<rgba 0.5 0.3086 0.1428 1>
   ) {isSetParameters=true, module=SetVariables_#4}
   /SetVariables(
    Particles.FrontCam Offset:NiagaraFloat=&Add_Float(A=0.1, B=$Particles.CameraOffset)
    Particles.FrontSize:Vector2f=&Multiply_Vector2D_ByFloat(Float=0.9, Vector2D=$Particles.SpriteSize)
    Particles.Frontcolor:LinearColor=<rgba 0.75 0.75 0.015 0.35>
   ) {isSetParameters=true, module=SetVariables_#5}
  ]
  rendererClasses [&NiagaraSpriteRendererProperties]
  renderers [{rendererClass=&NiagaraSpriteRendererProperties}]
 }
}
summary {
 systemName=NS_BananaPhilosopher
 emitters [
  {emitterName=Sparkles, rendererClasses [&NiagaraSpriteRendererProperties]}
  {bEnabled=false, emitterName=Sparkles001, rendererClasses [&NiagaraSpriteRendererProperties]}
 ]
 userVariables []
}
