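# Control Rig for the Banana Philosopher face accessory that applies corrective offsets to chest and beard bones as the upper arms move.
# A sequence runs one Spherical Pose Reader per side, reading upperarm_l and upperarm_r relative to spine_05 with asymmetric active and falloff regions.
# Each reader's output weights a Modify Transforms node that offsets that side's bones, including dyn_chest_fix and dyn_beard_01.
# Debug drawing remains enabled on the left reader.
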ControlRigBlueprint /Game/Characters/Player/Male/Medium/Bodies/M_MED_BananaPhilosopher/Meshes/Parts/M_MED_BananaPhilosopher_FaceAcc_CtrlRig
graphs [
 ContainedGraph {nodes [
   n1 = RigVMFunctionEntryNode()
   SequenceAggregate = &RigVMFunction_Sequence(ExecuteContext=@n1.ExecuteContext)
   SequenceAggregate_1 = &RigVMFunction_Sequence(ExecuteContext=@SequenceAggregate.B)
   SequenceAggregate_2 = &RigVMFunction_Sequence(ExecuteContext=@SequenceAggregate_1.B)
   RigVMFunctionReturnNode(A=@SequenceAggregate.A, B=@SequenceAggregate_1.A, C=@SequenceAggregate_2.A, D=@SequenceAggregate_2.B)
  ]}
 RigVMFunctionLibrary {links [], nodes []}
 RigVMModel {nodes [
   RigUnit_BeginExecution = &RigUnit_BeginExecution()
   SequenceAggregate = RigVMAggregateNode(ExecuteContext=@RigUnit_BeginExecution.ExecutePin)
   n1 = &RigUnit_SphericalPoseReader(
    ActiveRegionScaleFactors=(PositiveWidth=1,NegativeWidth=2.5,PositiveHeight=2.5,NegativeHeight=0.1)
    ActiveRegionSize=0.1
    Debug=(bDrawDebug=true,DebugScale=25,DebugSegments=20,DebugThickness=0.25)
    DriverAxis=(X=1,Y=0,Z=0)
    DriverItem=(Type=Bone,Name="upperarm_l")
    FalloffRegionScaleFactors=(PositiveWidth=1,NegativeWidth=2,PositiveHeight=2,NegativeHeight=1)
    FalloffSize=0.38
    FlipWidthScaling=true
    OptionalParentItem=(Type=Bone,Name="spine_05")
    RotationOffset=(X=90,Y=0,Z=35)
    ExecutePin=@SequenceAggregate.A
   )
   > &RigUnit_ModifyTransforms(
    ItemToModify="((Item=(Type=Bone,Name=\"dyn_chest_fix_l\"),Transform=(Rotation=(X=0.9659,Y=0,Z=0,W=0.2588),Translation=(X=10,Y=0,Z=0),Scale3D=(X=1,Y=1,Z=1))),(Item=(Type=Bone,Name=\"dyn_beard_01_l\"),Transform=(Rotation=(X=0,Y=0,Z=0,W=1),Translation=(X=-5,Y=0,Z=10),Scale3D=( ...(truncated)"
    Weight=@n1.OutputParam
   )
   n2 = &RigUnit_SphericalPoseReader(
    ActiveRegionScaleFactors=(PositiveWidth=1,NegativeWidth=2.5,PositiveHeight=2.5,NegativeHeight=0.1)
    ActiveRegionSize=0.1
    Debug=(bDrawDebug=false,DebugScale=25,DebugSegments=20,DebugThickness=0.25)
    DriverAxis=(X=-1,Y=0,Z=0)
    DriverItem=(Type=Bone,Name="upperarm_r")
    FalloffRegionScaleFactors=(PositiveWidth=1,NegativeWidth=2,PositiveHeight=2,NegativeHeight=0.7)
    FalloffSize=0.38
    FlipWidthScaling=false
    OptionalParentItem=(Type=Bone,Name="spine_05")
    RotationOffset=(X=-90,Y=0,Z=35)
    ExecutePin=@SequenceAggregate.B
   )
   > &RigUnit_ModifyTransforms(
    ItemToModify="((Item=(Type=Bone,Name=\"dyn_chest_fix_r\"),Transform=(Rotation=(X=0,Y=0,Z=0.4226,W=0.9063),Translation=(X=-30,Y=0,Z=0),Scale3D=(X=1,Y=1,Z=1))),(Item=(Type=Bone,Name=\"dyn_beard_01_r\"),Transform=(Rotation=(X=0,Y=0,Z=0,W=1),Translation=(X=5,Y=0,Z=-7),Scale3D=( ...(truncated)"
    Weight=@n2.OutputParam
   )
  ]}
]
hierarchy [
 root {loc=<0 0 0>, rot=<0 0 0 1>}
 pelvis {parent=root, loc=<0 -0.7071 93.1191>, rot=<0.0127 0.707 -0.0127 -0.707>}
 spine_01 {parent=pelvis, loc=<6.7494 -0.6388 0>, rot=<0 0 0.034 -0.9994>}
 spine_02 {parent=spine_01, loc=<7.2518 -0.035 0>, rot=<0 0 0.0238 -0.9997>}
 spine_03 {parent=spine_02, loc=<6.0104 0 0>, rot=<0 0 0.0149 -0.9999>}
 spine_04 {parent=spine_03, loc=<8.1024 0 0>, rot=<0 0 -0.0132 -0.9999>}
 spine_05 {parent=spine_04, loc=<10.0146 0 0>, rot=<0 0 -0.0015 -1>}
 neck_01 {parent=spine_05, loc=<11.3138 0.1529 0>, rot=<0 0 -0.2003 -0.9797>}
 neck_02 {parent=neck_01, loc=<4.5627 0 0>, rot=<0 0 0.0504 -0.9987>}
 head {parent=neck_02, loc=<5.3914 0 0>, rot=<0 0 0.115 -0.9934>}
 head_FX {parent=head, loc=<34.4077 5.1548 0>, rot=<0 0 0 1>}
 clavicle_l {parent=spine_05, loc=<4.3575 -0.1466 -2.3223>, rot=<0.0465 -0.7289 -0.009 -0.6829>}
 upperarm_l {parent=clavicle_l, loc=<15.5111 0 0>, rot=<0.0226 -0.3234 -0.0088 -0.9459>}
 upperarm_twist_01_l {parent=upperarm_l, loc=<0.0041 0 0>, rot=<0 0 0 1>}
 upperarm_twist_02_l {parent=upperarm_l, loc=<13.7941 0 0>, rot=<0 0 0 1>}
 clavicle_r {parent=spine_05, loc=<4.3575 -0.1466 2.3223>, rot=<0.7289 0.0465 -0.6829 0.009>}
 upperarm_r {parent=clavicle_r, loc=<-15.5111 0 0>, rot=<0.0226 -0.3234 -0.0088 -0.9459>}
 upperarm_twist_01_r {parent=upperarm_r, loc=<-0.0041 0 0>, rot=<0 0 0 -1>}
 upperarm_twist_02_r {parent=upperarm_r, loc=<-13.7941 0 0>, rot=<0 0 0 -1>}
 dyn_banana_peel_01 {parent=spine_05, loc=<55.7024 -2.5987 0.109>, rot=<0.1007 -0.6977 0.7014 -0.1059>}
 dyn_banana_peel_02 {parent=dyn_banana_peel_01, loc=<8.3589 0.1727 4.2778>, rot=<0 -0.0565 0 -0.9984>}
 dyn_banana_peel_03 {parent=dyn_banana_peel_02, loc=<9.3628 -0.154 3.2984>, rot=<0 0 0 1>}
 dyn_banana_peel_r_01 {parent=spine_05, loc=<50.476 10.8664 10.2929>, rot=<0.134 -0.1044 0.9757 -0.1382>}
 dyn_banana_peel_r_02 {parent=dyn_banana_peel_r_01, loc=<7.8767 -1.1942 2.667>, rot=<0 -0.0565 0 -0.9984>}
 dyn_banana_peel_r_03 {parent=dyn_banana_peel_r_02, loc=<10.4598 -0.9792 1.9811>, rot=<0 0 0 1>}
 dyn_banana_peel_l_01 {parent=spine_05, loc=<50.476 10.8664 -10.2929>, rot=<0.0357 -0.9916 -0.0045 -0.1246>}
 dyn_banana_peel_l_02 {parent=dyn_banana_peel_l_01, loc=<7.8767 -1.1942 2.667>, rot=<0 -0.0565 0 -0.9984>}
 dyn_banana_peel_l_03 {parent=dyn_banana_peel_l_02, loc=<10.4598 -0.9792 1.9811>, rot=<0 0 0 1>}
 dyn_chest_fix_l {parent=spine_05, loc=<7.3424 9.218 -16.5505>, rot=<0.0465 -0.7289 -0.009 -0.6829>}
 dyn_chest_fix_r {parent=spine_05, loc=<7.3424 9.218 16.5505>, rot=<-0.026 0.7299 -0.0281 -0.6824>}
 dyn_beard_00 {parent=spine_05, loc=<7.2836 14.1563 0>, rot=<0 0 0.9891 0.1472>}
 dyn_beard_01 {parent=dyn_beard_00, loc=<4.2744 -1.7064 0>, rot=<0 0 0.1649 -0.9863>}
 dyn_beard_02 {parent=dyn_beard_01, loc=<5.9237 0 0>, rot=<0 0 -0.1263 -0.992>}
 dyn_beard_02_l {parent=dyn_beard_02, loc=<0 0 -10>, rot=<0 0 0 1>}
 dyn_beard_02_r {parent=dyn_beard_02, loc=<0 0 10>, rot=<0 0 0 1>}
 dyn_beard_01_l {parent=dyn_beard_01, loc=<0 0 -12>, rot=<0 0 0 1>}
 dyn_beard_01_r {parent=dyn_beard_01, loc=<0 0 12>, rot=<0 0 0 1>}
 gravity {parent=root, loc=<0 0 0>, rot=<0 0 0 1>}
]
