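# Physics asset for the face accessory of the BananaPhilosopher player character, giving capsule bodies to its dynamic banana peel and beard bones.
# Three three-bone peel chains (center, left, right) hang from a large spine_05 capsule, with equal swing and twist limits at each joint.
# The beard branches left and right at two levels below a tiny root body.
# That root's angular motion to spine_05 is free, while lower joints have tight twist limits.
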PhysicsAsset /Game/Characters/Player/Male/Medium/Bodies/M_MED_BananaPhilosopher/Meshes/Parts/M_MED_BananaPhilosopher_FaceAcc_Physics
bodies {
 dyn_banana_peel_01 [{
   shapeType=Capsule
   length=10.2138
   radius=4.8781
   shapeName=dyn_banana_peel_01_capsule
   center=<4.6238 0 2.3301>
   rotation=<rot 65.0002 -180 -180>
  }]
 dyn_banana_peel_02 [{
   shapeType=Capsule
   length=10.2138
   radius=4.8781
   shapeName=dyn_banana_peel_02_capsule
   center=<5.1944 0 1.5701>
   rotation=<rot 75.0002 -180 -180>
  }]
 dyn_banana_peel_03 [{
   shapeType=Capsule
   length=10.2138
   radius=4.8781
   shapeName=dyn_banana_peel_03_capsule
   center=<2.8845 0 3.759>
   rotation=<rot 35.0005 180 -180>
  }]
 dyn_banana_peel_l_01 [{
   shapeType=Capsule
   length=11.0096
   radius=4.0823
   shapeName=dyn_banana_peel_l_01_capsule
   center=<5.4865 0 0.48>
   rotation=<rot 85 -179.9996 -179.9999>
  }]
 dyn_banana_peel_l_02 [{
   shapeType=Capsule
   length=11.6146
   radius=3.4773
   shapeName=dyn_banana_peel_l_02_capsule
   center=<5.4865 0 0.48>
   rotation=<rot 85 -179.9997 -179.9999>
  }]
 dyn_banana_peel_l_03 [{
   shapeType=Capsule
   length=11.6146
   radius=3.4773
   shapeName=dyn_banana_peel_l_03_capsule
   center=<2.3821 0 2.6537>
   rotation=<rot 35 -179.9997 -180>
  }]
 dyn_banana_peel_r_01 [{
   shapeType=Capsule
   length=11.0096
   radius=4.0823
   shapeName=dyn_banana_peel_r_01_capsule
   center=<5.4865 0 0.48>
   rotation=<rot 85 179.9997 179.9999>
  }]
 dyn_banana_peel_r_02 [{
   shapeType=Capsule
   length=11.6146
   radius=3.4773
   shapeName=dyn_banana_peel_r_02_capsule
   center=<5.4865 0 0.48>
   rotation=<rot 85 179.9997 179.9999>
  }]
 dyn_banana_peel_r_03 [{
   shapeType=Capsule
   length=11.6146
   radius=3.4773
   shapeName=dyn_banana_peel_r_03_capsule
   center=<3.0449 0 2.5288>
   rotation=<rot 40 179.9996 180>
  }]
 dyn_beard_00 [{shapeType=Capsule, length=0.505, radius=0.505, shapeName=dyn_beard_00_capsule, rotation=<rot 0 -90 0>}]
 dyn_beard_01 [{
   shapeType=Capsule
   length=12.3689
   radius=7.4556
   shapeName=dyn_beard_01_capsule
   center=<3.0785 0.7385 -0.1799>
   rotation=<rot 7.6158 -90.0001 -3.3498>
  }]
 dyn_beard_01_l [{
   shapeType=Capsule
   length=2.7151
   radius=4.5172
   shapeName=dyn_beard_01_l_capsule
   center=<1.3688 2.8316 -1.135>
   rotation=<rot 0 83.8681 -50.5181>
  }]
 dyn_beard_01_r [{
   shapeType=Capsule
   length=2.638
   radius=4.6907
   shapeName=dyn_beard_01_r_capsule
   center=<2.082 2.3526 0.9945>
   rotation=<rot 0 -170.2814 -67.6402>
  }]
 dyn_beard_02 [{
   shapeType=Capsule
   length=8.2688
   radius=6.6902
   shapeName=dyn_beard_02_capsule
   center=<3.3852 0.4628 -0.7458>
   rotation=<rot -2.7434 -90.0001 -0.3148>
  }]
 dyn_beard_02_l [{
   shapeType=Capsule
   length=4.3719
   radius=5.0502
   shapeName=dyn_beard_02_l_capsule
   center=<3.4639 0.4853 0.6059>
   rotation=<rot 0 -10.2346 -76.9108>
  }]
 dyn_beard_02_r [{
   shapeType=Capsule
   length=4.2414
   radius=4.5613
   shapeName=dyn_beard_02_r_capsule
   center=<2.834 0.8546 -0.6646>
   rotation=<rot 0 161.8372 -76.8483>
  }]
 spine_05 [{shapeType=Capsule, length=41.0678, radius=13.4778, center=<21.0583 -0.6311 -0.0962>, rotation=<rot -0.2625 -73.5501 -89.625>}]
}
constraints [
 {bone1Name=dyn_banana_peel_01, bone2Name=spine_05, swing1LimitDegrees=45, swing2LimitDegrees=45, twistLimitDegrees=45}
 {bone1Name=dyn_banana_peel_r_01, bone2Name=spine_05, swing1LimitDegrees=45, swing2LimitDegrees=45, twistLimitDegrees=45}
 {bone1Name=dyn_banana_peel_l_01, bone2Name=spine_05, swing1LimitDegrees=45, swing2LimitDegrees=45, twistLimitDegrees=45}
 {bone1Name=dyn_banana_peel_02, bone2Name=dyn_banana_peel_01, swing1LimitDegrees=45, swing2LimitDegrees=45, twistLimitDegrees=45}
 {bone1Name=dyn_banana_peel_03, bone2Name=dyn_banana_peel_02, swing1LimitDegrees=45, swing2LimitDegrees=45, twistLimitDegrees=45}
 {bone1Name=dyn_banana_peel_l_02, bone2Name=dyn_banana_peel_l_01, swing1LimitDegrees=45, swing2LimitDegrees=45, twistLimitDegrees=45}
 {bone1Name=dyn_banana_peel_l_03, bone2Name=dyn_banana_peel_l_02, swing1LimitDegrees=45, swing2LimitDegrees=45, twistLimitDegrees=45}
 {bone1Name=dyn_banana_peel_r_02, bone2Name=dyn_banana_peel_r_01, swing1LimitDegrees=45, swing2LimitDegrees=45, twistLimitDegrees=45}
 {bone1Name=dyn_banana_peel_r_03, bone2Name=dyn_banana_peel_r_02, swing1LimitDegrees=45, swing2LimitDegrees=45, twistLimitDegrees=45}
 {
  bone1Name=dyn_beard_00
  bone2Name=spine_05
  swing1LimitDegrees=45
  swing1Motion=Free
  swing2LimitDegrees=45
  swing2Motion=Free
  twistLimitDegrees=45
  twistMotion=Free
 }
 {bone1Name=dyn_beard_01, bone2Name=dyn_beard_00, swing1LimitDegrees=45, swing2LimitDegrees=10, twistLimitDegrees=1}
 {bone1Name=dyn_beard_02, bone2Name=dyn_beard_01, swing1LimitDegrees=45, swing2LimitDegrees=10, twistLimitDegrees=1}
 {bone1Name=dyn_beard_02_l, bone2Name=dyn_beard_02, swing1LimitDegrees=45, swing2LimitDegrees=10, twistLimitDegrees=1}
 {bone1Name=dyn_beard_02_r, bone2Name=dyn_beard_02, swing1LimitDegrees=45, swing2LimitDegrees=10, twistLimitDegrees=1}
 {bone1Name=dyn_beard_01_l, bone2Name=dyn_beard_01, swing1LimitDegrees=45, swing2LimitDegrees=10, twistLimitDegrees=1}
 {bone1Name=dyn_beard_01_r, bone2Name=dyn_beard_01, swing1LimitDegrees=45, swing2LimitDegrees=10, twistLimitDegrees=1}
]
