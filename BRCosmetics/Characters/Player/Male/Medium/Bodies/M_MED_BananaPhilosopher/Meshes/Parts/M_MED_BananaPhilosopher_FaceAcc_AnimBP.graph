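# Post-process-style AnimBlueprint for the Banana Philosopher face accessory, animating its facial idle and beard.
# It copies the character mesh pose, cached on customization load, and layers the banana facial idle by 1 minus DisableFaceOverride.
# Weighted constraints position the gravity, chest-fix and beard bones from torso bones.
# Per-bone Anim Dynamics springs relative to the gravity bone, a RigidBody simulation and a Control Rig finish the pose.
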AnimBlueprint /BRCosmetics/Characters/Player/Male/Medium/Bodies/M_MED_BananaPhilosopher/Meshes/Parts/M_MED_BananaPhilosopher_FaceAcc_AnimBP
target_skeleton=/Game/Characters/Player/Male/Medium/Bodies/M_MED_BananaPhilosopher/Meshes/Parts/M_MED_BananaPhilosopher_FaceAcc_Skeleton
graphs [
 AnimGraph {nodes [
   n1 = Copy Pose From Mesh(SourceMeshComponent=Get Character Mesh() -> (Character Mesh:Skeletal Mesh Component Object Reference)) {bCopyCurves=true, bUseAttachedParent=true}
   n2 = Sequence Player 'Banana_Male_Facial_Idle'() {sequence=/Game/Animation/Game/MainPlayer/Facial/Banana/Banana_Male_Facial_Idle}
   n3 = float - float(A=1, B=Get DisableFaceOverride() -> (DisableFaceOverride:float))
   n4 = "Constraint - gravity (TT)"!(
    ComponentPose=Local To Component(LocalPose=Layered blend per bone(BasePose=@n1, BlendPoses_0=@n2, BlendWeights_0=@n3))
    Alpha=1
    ConstraintWeights_0=0
    ConstraintWeights_1:float=1
   ) -> (Pose) {
    alphaBoolBlend {blendOption=Linear}
    boneToModify {boneName=gravity}
    constraintSetup [{transformType=Translation, targetBone {boneName=root}}, {transformType=Translation, targetBone {boneName=spine_05}}]
    constraintWeights [1,1]
   }
   n5 = "Constraint - gravity (RR)"!(ComponentPose=@n4, Alpha=1, ConstraintWeights_0=0, ConstraintWeights_1:float=1) -> (Pose) {
    alphaBoolBlend {blendOption=Linear}
    boneToModify {boneName=gravity}
    constraintSetup [{transformType=Rotation, targetBone {boneName=root}}, {transformType=Rotation, targetBone {boneName=spine_05}}]
    constraintWeights [1,1]
   }
   n6 = "Constraint - dyn_chest_fix_l (PPPPP)"!(
    ComponentPose=@n5
    Alpha=1
    ConstraintWeights_0=0.5231
    ConstraintWeights_1:float=0.2514
    ConstraintWeights_2:float=0.0988
    ConstraintWeights_3:float=0.0841
    ConstraintWeights_4:float=0.0426
   ) -> (Pose) {
    alphaBoolBlend {blendOption=Linear}
    boneToModify {boneName=dyn_chest_fix_l}
    constraintSetup [
     {targetBone {boneName=spine_05}}
     {targetBone {boneName=clavicle_l}}
     {targetBone {boneName=spine_04}}
     {targetBone {boneName=upperarm_twist_01_l}}
     {targetBone {boneName=spine_03}}
    ]
    constraintWeights [1,1,1,1,1]
   }
   n7 = "Constraint - dyn_chest_fix_r (PPPPP)"!(
    ComponentPose=@n6
    Alpha=1
    ConstraintWeights_0=0.5231
    ConstraintWeights_1:float=0.2514
    ConstraintWeights_2:float=0.0988
    ConstraintWeights_3:float=0.0841
    ConstraintWeights_4:float=0.0426
   ) -> (Pose) {
    alphaBoolBlend {blendOption=Linear}
    boneToModify {boneName=dyn_chest_fix_r}
    constraintSetup [
     {targetBone {boneName=spine_05}}
     {targetBone {boneName=clavicle_r}}
     {targetBone {boneName=spine_04}}
     {targetBone {boneName=upperarm_twist_01_r}}
     {targetBone {boneName=spine_03}}
    ]
    constraintWeights [1,1,1,1,1]
   }
   n8 = "Constraint - dyn_beard_00 (PPPPP)"!(
    ComponentPose=@n7
    Alpha=1
    ConstraintWeights_0=0.5691
    ConstraintWeights_1:float=0.27
    ConstraintWeights_2:float=0.0907
    ConstraintWeights_3:float=0.0503
    ConstraintWeights_4:float=0.0198
   ) -> (Pose) {
    alphaBoolBlend {blendOption=Linear}
    boneToModify {boneName=dyn_beard_00}
    constraintSetup [
     {targetBone {boneName=spine_05}}
     {targetBone {boneName=spine_04}}
     {targetBone {boneName=spine_03}}
     {targetBone {boneName=neck_01}}
     {targetBone {boneName=spine_02}}
    ]
    constraintWeights [1,1,1,1,1]
   }
   n9 = "Constraint - dyn_beard_01 (PPPPPPPP)"!(
    ComponentPose=@n8
    Alpha=1
    ConstraintWeights_0=0.432
    ConstraintWeights_1:float=0.3004
    ConstraintWeights_2:float=0.1421
    ConstraintWeights_3:float=0.0451
    ConstraintWeights_4:float=0.0364
    ConstraintWeights_5:float=0.0353
    ConstraintWeights_6:float=0.0064
    ConstraintWeights_7:float=0.0023
   ) -> (Pose) {
    alphaBoolBlend {blendOption=Linear}
    boneToModify {boneName=dyn_beard_01}
    constraintSetup [
     {targetBone {boneName=spine_04}}
     {targetBone {boneName=spine_05}}
     {targetBone {boneName=spine_03}}
     {targetBone {boneName=clavicle_l}}
     {targetBone {boneName=spine_02}}
     {targetBone {boneName=clavicle_r}}
     {targetBone {boneName=neck_02}}
     {targetBone {boneName=neck_01}}
    ]
    constraintWeights [1,1,1,1,1,1,1,1]
   }
   n10 = "Constraint - dyn_beard_02 (PPPPPP)"!(
    ComponentPose=@n9
    Alpha=1
    ConstraintWeights_0=0.5097
    ConstraintWeights_1:float=0.2673
    ConstraintWeights_2:float=0.1647
    ConstraintWeights_3:float=0.0493
    ConstraintWeights_4:float=0.0057
    ConstraintWeights_5:float=0.0033
   ) -> (Pose) {
    alphaBoolBlend {blendOption=Linear}
    boneToModify {boneName=dyn_beard_02}
    constraintSetup [
     {targetBone {boneName=spine_04}}
     {targetBone {boneName=spine_05}}
     {targetBone {boneName=spine_03}}
     {targetBone {boneName=spine_02}}
     {targetBone {boneName=clavicle_l}}
     {targetBone {boneName=clavicle_r}}
    ]
    constraintWeights [1,1,1,1,1,1]
   }
   n11 = "Constraint - dyn_beard_02_l (PPPPP)"!(
    ComponentPose=@n10
    Alpha=1
    ConstraintWeights_0=0.4783
    ConstraintWeights_1:float=0.3434
    ConstraintWeights_2:float=0.1442
    ConstraintWeights_3:float=0.0259
    ConstraintWeights_4:float=0.0081
   ) -> (Pose) {
    alphaBoolBlend {blendOption=Linear}
    boneToModify {boneName=dyn_beard_02_l}
    constraintSetup [
     {targetBone {boneName=spine_04}}
     {targetBone {boneName=spine_05}}
     {targetBone {boneName=spine_03}}
     {targetBone {boneName=spine_02}}
     {targetBone {boneName=clavicle_l}}
    ]
    constraintWeights [1,1,1,1,1]
   }
   n12 = "Constraint - dyn_beard_02_r (PPPPP)"!(
    ComponentPose=@n11
    Alpha=1
    ConstraintWeights_0=0.4783
    ConstraintWeights_1:float=0.3434
    ConstraintWeights_2:float=0.1442
    ConstraintWeights_3:float=0.0259
    ConstraintWeights_4:float=0.0081
   ) -> (Pose) {
    alphaBoolBlend {blendOption=Linear}
    boneToModify {boneName=dyn_beard_02_r}
    constraintSetup [
     {targetBone {boneName=spine_04}}
     {targetBone {boneName=spine_05}}
     {targetBone {boneName=spine_03}}
     {targetBone {boneName=spine_02}}
     {targetBone {boneName=clavicle_r}}
    ]
    constraintWeights [1,1,1,1,1]
   }
   n13 = "Constraint - dyn_beard_01_l (PPPPPP)"!(
    ComponentPose=@n12
    Alpha=1
    ConstraintWeights_0=0.4094
    ConstraintWeights_1:float=0.3671
    ConstraintWeights_2:float=0.1244
    ConstraintWeights_3:float=0.0726
    ConstraintWeights_4:float=0.0164
    ConstraintWeights_5:float=0.0102
   ) -> (Pose) {
    alphaBoolBlend {blendOption=Linear}
    boneToModify {boneName=dyn_beard_01_l}
    constraintSetup [
     {targetBone {boneName=spine_04}}
     {targetBone {boneName=spine_05}}
     {targetBone {boneName=spine_03}}
     {targetBone {boneName=clavicle_l}}
     {targetBone {boneName=upperarm_twist_01_l}}
     {targetBone {boneName=spine_02}}
    ]
    constraintWeights [1,1,1,1,1,1]
   }
   n14 = "Constraint - dyn_beard_01_r (PPPPPP)"!(
    ComponentPose=@n13
    Alpha=1
    ConstraintWeights_0=0.4094
    ConstraintWeights_1:float=0.3671
    ConstraintWeights_2:float=0.1244
    ConstraintWeights_3:float=0.0726
    ConstraintWeights_4:float=0.0164
    ConstraintWeights_5:float=0.0102
   ) -> (Pose) {
    alphaBoolBlend {blendOption=Linear}
    boneToModify {boneName=dyn_beard_01_r}
    constraintSetup [
     {targetBone {boneName=spine_04}}
     {targetBone {boneName=spine_05}}
     {targetBone {boneName=spine_03}}
     {targetBone {boneName=clavicle_r}}
     {targetBone {boneName=upperarm_twist_01_r}}
     {targetBone {boneName=spine_02}}
    ]
    constraintWeights [1,1,1,1,1,1]
   }
   n15 = Anim Dynamics - Bone: dyn_beard_00!(ComponentPose=@n14, Alpha=1, GravityOverride:Vector="0,0,-980", ExternalForce="0,0,0") -> (Pose) {
    angularDampingOverride=0.99
    angularSpringConstant=800
    bAngularSpring=true
    bOverrideAngularDamping=true
    bUseGravityOverride=true
    gravityScale=0.3
    simulationSpace=BoneRelative
    alphaBoolBlend {blendOption=Linear}
    boundBone {boneName=dyn_beard_00}
    componentAppliedLinearAccClamp=<800 800 400>
    componentLinearAccScale=<0.2 0.2 0.2>
    componentLinearVelScale=<0.21 0.21 0.21>
    gravityOverride {z=500}
    physicsBodyDefinitions [{
      collisionType=CustomSphere
      sphereCollisionRadius=2
      boundBone {boneName=dyn_beard_00}
      boxExtents=<50 50 50>
      constraintSetup {angularLimitsMax=<0 0 0>, angularLimitsMin=<0 0 -20>, angularTarget=<0 0 0>}
      localJointOffset=<5 1 0>
     }]
    relativeSpaceBone {boneName=gravity}
   }
   n16 = Anim Dynamics - Bone: dyn_beard_01!(ComponentPose=@n15, Alpha=1, GravityOverride:Vector="0,0,-980", ExternalForce="0,0,0") -> (Pose) {
    angularDampingOverride=0.99
    angularSpringConstant=800
    bAngularSpring=true
    bOverrideAngularDamping=true
    bUseGravityOverride=true
    gravityScale=0.3
    simulationSpace=BoneRelative
    alphaBoolBlend {blendOption=Linear}
    boundBone {boneName=dyn_beard_01}
    componentAppliedLinearAccClamp=<800 800 400>
    componentLinearAccScale=<0.2 0.2 0.2>
    componentLinearVelScale=<0.21 0.21 0.21>
    gravityOverride {z=500}
    physicsBodyDefinitions [{
      collisionType=CustomSphere
      sphereCollisionRadius=2
      boundBone {boneName=dyn_beard_01}
      boxExtents=<33 33 33>
      constraintSetup {angularLimitsMax=<0 0 0>, angularLimitsMin=<0 0 -20>, angularTarget=<0 0 0>}
      localJointOffset=<5 1 0>
     }]
    relativeSpaceBone {boneName=gravity}
   }
   n17 = Anim Dynamics - Bone: dyn_beard_01_l!(ComponentPose=@n16, Alpha=1, GravityOverride:Vector="0,0,-980", ExternalForce="0,0,0") -> (Pose) {
    angularDampingOverride=0.99
    angularSpringConstant=800
    bAngularSpring=true
    bOverrideAngularDamping=true
    bUseGravityOverride=true
    gravityScale=0.3
    simulationSpace=BoneRelative
    alphaBoolBlend {blendOption=Linear}
    boundBone {boneName=dyn_beard_01_l}
    componentAppliedLinearAccClamp=<800 800 400>
    componentLinearAccScale=<0.2 0.2 0.2>
    componentLinearVelScale=<0.21 0.21 0.21>
    gravityOverride {z=500}
    physicsBodyDefinitions [{
      collisionType=CustomSphere
      sphereCollisionRadius=2
      boundBone {boneName=dyn_beard_01_l}
      boxExtents=<33 33 33>
      constraintSetup {angularLimitsMax=<0 0 0>, angularLimitsMin=<0 0 -20>, angularTarget=<0 0 0>}
      localJointOffset=<5 1 0>
     }]
    relativeSpaceBone {boneName=gravity}
   }
   n18 = Anim Dynamics - Bone: dyn_beard_01_r!(ComponentPose=@n17, Alpha=1, GravityOverride:Vector="0,0,-980", ExternalForce="0,0,0") -> (Pose) {
    angularDampingOverride=0.99
    angularSpringConstant=800
    bAngularSpring=true
    bOverrideAngularDamping=true
    bUseGravityOverride=true
    gravityScale=0.3
    simulationSpace=BoneRelative
    alphaBoolBlend {blendOption=Linear}
    boundBone {boneName=dyn_beard_01_r}
    componentAppliedLinearAccClamp=<800 800 400>
    componentLinearAccScale=<0.2 0.2 0.2>
    componentLinearVelScale=<0.21 0.21 0.21>
    gravityOverride {z=500}
    physicsBodyDefinitions [{
      collisionType=CustomSphere
      sphereCollisionRadius=2
      boundBone {boneName=dyn_beard_01_r}
      boxExtents=<33 33 33>
      constraintSetup {angularLimitsMax=<0 0 0>, angularLimitsMin=<0 0 -20>, angularTarget=<0 0 0>}
      localJointOffset=<5 1 0>
     }]
    relativeSpaceBone {boneName=gravity}
   }
   n19 = Anim Dynamics - Bone: dyn_beard_02!(ComponentPose=@n18, Alpha=1, GravityOverride:Vector="0,0,-980", ExternalForce="0,0,0") -> (Pose) {
    angularDampingOverride=0.99
    angularSpringConstant=800
    bAngularSpring=true
    bOverrideAngularDamping=true
    bUseGravityOverride=true
    gravityScale=0.3
    simulationSpace=BoneRelative
    alphaBoolBlend {blendOption=Linear}
    boundBone {boneName=dyn_beard_02}
    componentAppliedLinearAccClamp=<800 800 400>
    componentLinearAccScale=<0.2 0.2 0.2>
    componentLinearVelScale=<0.21 0.21 0.21>
    gravityOverride {z=500}
    physicsBodyDefinitions [{
      collisionType=CustomSphere
      sphereCollisionRadius=2
      boundBone {boneName=dyn_beard_02}
      boxExtents=<27 27 27>
      constraintSetup {angularLimitsMax=<0 0 0>, angularLimitsMin=<0 0 -20>, angularTarget=<0 0 0>}
      localJointOffset=<5 1 0>
     }]
    relativeSpaceBone {boneName=gravity}
   }
   n20 = Anim Dynamics - Bone: dyn_beard_02_l!(ComponentPose=@n19, Alpha=1, GravityOverride:Vector="0,0,-980", ExternalForce="0,0,0") -> (Pose) {
    angularDampingOverride=0.99
    angularSpringConstant=800
    bAngularSpring=true
    bOverrideAngularDamping=true
    bUseGravityOverride=true
    gravityScale=0.3
    simulationSpace=BoneRelative
    alphaBoolBlend {blendOption=Linear}
    boundBone {boneName=dyn_beard_02_l}
    componentAppliedLinearAccClamp=<800 800 400>
    componentLinearAccScale=<0.2 0.2 0.2>
    componentLinearVelScale=<0.21 0.21 0.21>
    gravityOverride {z=500}
    physicsBodyDefinitions [{
      collisionType=CustomSphere
      sphereCollisionRadius=2
      boundBone {boneName=dyn_beard_02_l}
      boxExtents=<27 27 27>
      constraintSetup {angularLimitsMax=<0 0 0>, angularLimitsMin=<0 0 -20>, angularTarget=<0 0 0>}
      localJointOffset=<5 1 0>
     }]
    relativeSpaceBone {boneName=gravity}
   }
   n21 = Anim Dynamics - Bone: dyn_beard_02_r!(ComponentPose=@n20, Alpha=1, GravityOverride:Vector="0,0,-980", ExternalForce="0,0,0") -> (Pose) {
    angularDampingOverride=0.99
    angularSpringConstant=800
    bAngularSpring=true
    bOverrideAngularDamping=true
    bUseGravityOverride=true
    gravityScale=0.3
    simulationSpace=BoneRelative
    alphaBoolBlend {blendOption=Linear}
    boundBone {boneName=dyn_beard_02_r}
    componentAppliedLinearAccClamp=<800 800 400>
    componentLinearAccScale=<0.2 0.2 0.2>
    componentLinearVelScale=<0.21 0.21 0.21>
    gravityOverride {z=500}
    physicsBodyDefinitions [{
      collisionType=CustomSphere
      sphereCollisionRadius=2
      boundBone {boneName=dyn_beard_02_r}
      boxExtents=<27 27 27>
      constraintSetup {angularLimitsMax=<0 0 0>, angularLimitsMin=<0 0 -20>, angularTarget=<0 0 0>}
      localJointOffset=<5 1 0>
     }]
    relativeSpaceBone {boneName=gravity}
   }
   n22 = RigidBody!(
    ComponentPose=@n21
    Alpha
    OverrideWorldGravity:Vector="0,0,-250"
    ExternalForce
    ComponentLinearAccScale:Vector="0.25,0.35,0.35"
    ComponentLinearVelScale:Vector="0.3,0.4,0.4"
   ) -> (Pose) {
    bClampLinearTranslationLimitToRefPose=true
    bOverrideWorldGravity=true
    overridePhysicsAsset=/Game/Characters/Player/Male/Medium/Bodies/M_MED_BananaPhilosopher/Meshes/Parts/M_MED_BananaPhilosopher_FaceAcc_Physics
    simulationSpace=BaseBoneSpace
    alphaBoolBlend {blendOption=Linear}
    baseBoneRef {boneName=spine_03}
    componentAppliedLinearAccClamp=<700 850 850>
    componentLinearAccScale=<0.334 0.334 0.334>
    componentLinearVelScale=<0.4 0.4 0.4>
    overrideWorldGravity {x=200, z=-980}
   }
   n23 = Control Rig(Source=Component To Local(ComponentPose=@n22)) {
    alphaBoolBlend {blendOption=Linear}
    controlRigAssetReference {blueprintRigClass=/Game/Characters/Player/Male/Medium/Bodies/M_MED_BananaPhilosopher/Meshes/Parts/M_MED_BananaPhilosopher_FaceAcc_CtrlRig.M_MED_BananaPhilosopher_FaceAcc_CtrlRig_C}
   }
   Output Pose(Result=@n23) {layerGroup=None}
  ]}
 EventGraph {nodes [
   n1 = Event OnCharacterCustomizationLoaded()
   > n2 = Sequence!(execute) -> (then_0, then_1, then_2:Exec, then_3:Exec)
   Set Character Mesh(execute=@n2.then_1, Character Mesh:Skeletal Mesh Component Object Reference=@n1.CustomBaseMesh) -> (Output_Get:Skeletal Mesh Component Object Reference)
   Set Body Mesh(execute=@n2.then_2, Body Mesh:Skeletal Mesh Component Object Reference=Get Array Item with EFort Custom Part Type(Array=@n1.CustomMeshParts)) -> (Output_Get:Skeletal Mesh Component Object Reference)
   GetSkeletalMesh!(self:Custom Character Part Object Reference=Get Array Item with EFort Custom Part Type(Array:Array of Custom Character Part Object References=@n1.CustomParts) -> (Output:Custom Character Part Object Reference)) -> (ReturnValue:Skeletal Mesh Object Reference)
   Event BlueprintUpdateAnimation()
  ]}
]
variables [
 Body Mesh {type=SkeletalMeshComponent}
 Character Mesh {type=SkeletalMeshComponent}
 skel Mesh {type=SkeletalMeshComponent}
]
